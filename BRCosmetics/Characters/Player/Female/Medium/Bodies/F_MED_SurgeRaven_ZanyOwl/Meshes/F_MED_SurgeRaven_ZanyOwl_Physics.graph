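# Physics asset for the female medium SurgeRaven ZanyOwl body, covering its dynamic jacket bones and a knife.
# Capsules on pelvis, spine and thighs form the body volumes, while tiny base capsules anchor the jacket to spine_01.
# Front flaps are three-link chains per side and back and side panels single links, all with tight swing and twist limits.
# Knife and base joints are free, except the locked right-front base, unlike its left twin.
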PhysicsAsset /BRCosmetics/Characters/Player/Female/Medium/Bodies/F_MED_SurgeRaven_ZanyOwl/Meshes/F_MED_SurgeRaven_ZanyOwl_Physics
bodies {
 base_f_jacket_l [{shapeType=Capsule, length=0.505, radius=0.505, shapeName=base_f_jacket_l_capsule}]
 base_f_jacket_r [{shapeType=Capsule, length=0.505, radius=0.505, shapeName=base_f_jacket_r_capsule}]
 base_jacket_b [{shapeType=Capsule, length=0.505, radius=0.505, shapeName=base_jacket_b_capsule, rotation=<rot 0 -90 0>}]
 base_jacket_l [{shapeType=Capsule, length=0.505, radius=0.505, shapeName=base_jacket_l_capsule, rotation=<rot 0 -90 0>}]
 base_jacket_r [{shapeType=Capsule, length=0.505, radius=0.505, shapeName=base_jacket_r_capsule, rotation=<rot 0 -90 0>}]
 dyn_f_jacket_01_l [{shapeType=Capsule, length=5, radius=4, shapeName=dyn_f_jacket_01_l_capsule, center=<4.0244 1.9992 -2.2465>, rotation=<rot 0 0 90>}]
 dyn_f_jacket_01_r [{shapeType=Capsule, length=5, radius=4, shapeName=dyn_f_jacket_01_r_capsule, center=<3.5038 3.4312 0>, rotation=<rot 0 0 90>}]
 dyn_f_jacket_02_l [{shapeType=Capsule, length=6, radius=4, shapeName=dyn_f_jacket_02_l_capsule, center=<4.8481 2.3688 -1.8981>, rotation=<rot 0 0 90>}]
 dyn_f_jacket_02_r [{shapeType=Capsule, length=7.461, radius=4.4721, shapeName=dyn_f_jacket_02_r_capsule, center=<4.695 4.0154 2.372>, rotation=<rot 0 0 90>}]
 dyn_f_jacket_l [{shapeType=Capsule, length=5, radius=4, shapeName=dyn_f_jacket_l_capsule, center=<5 2.4609 0>, rotation=<rot 0 0 90>}]
 dyn_f_jacket_r [{shapeType=Capsule, length=5, radius=4, shapeName=dyn_f_jacket_r_capsule, center=<5 3.4312 0>, rotation=<rot 0 0 90>}]
 dyn_jacket_b [{shapeType=Capsule, length=11.4, radius=3, shapeName=dyn_jacket_b_capsule, center=<8.202 0 4.1602>, rotation=<rot 0 0 90>}]
 dyn_jacket_l [{shapeType=Capsule, length=5, radius=4, shapeName=dyn_jacket_l_capsule, center=<3.17 7.2369 0>}]
 dyn_jacket_r [{shapeType=Capsule, length=9.5, radius=4, shapeName=dyn_jacket_r_capsule, center=<5 -5.8997 0>}]
 dyn_knife [{
   shapeType=Capsule
   length=32.1631
   radius=4.6022
   shapeName=dyn_knife_capsule
   center=<6.0853 2.2316 1.5985>
   rotation=<rot 0 80.9036 -80.0319>
  }]
 pelvis [
  {shapeType=Capsule, length=13.5, radius=8, shapeName=pelvis_capsule, rotation=<rot 0 -90 0>}
  {shapeType=Capsule, length=15.2, radius=10.1, shapeName=pelvis_capsule, center=<0 -2.0113 0>}
 ]
 spine_01 [{shapeType=Capsule, length=15, radius=10, shapeName=spine_01_capsule, center=<5.5119 0 0>, rotation=<rot 90 0 0>}]
 thigh_l [{
   shapeType=Capsule
   length=40.5
   radius=7.9529
   shapeName=thigh_l_capsule
   center=<-16.4716 -0.5191 0.1644>
   rotation=<rot 0.5579 -90.5989 -89.951>
  }]
 thigh_r [{
   shapeType=Capsule
   length=36.5
   radius=8.3183
   shapeName=thigh_r_capsule
   center=<16.6904 0.2775 -0.5538>
   rotation=<rot 0.4071 -91.6826 -93.5781>
  }]
}
constraints [
 {
  bone1Name=base_f_jacket_r
  bone2Name=spine_01
  swing1LimitDegrees=45
  swing1Motion=Locked
  swing2LimitDegrees=45
  swing2Motion=Locked
  twistLimitDegrees=45
  twistMotion=Locked
 }
 {
  bone1Name=base_f_jacket_l
  bone2Name=spine_01
  swing1LimitDegrees=45
  swing1Motion=Free
  swing2LimitDegrees=45
  swing2Motion=Free
  twistLimitDegrees=45
  twistMotion=Free
 }
 {
  bone1Name=thigh_l
  bone2Name=pelvis
  swing1LimitDegrees=45
  swing1Motion=Free
  swing2LimitDegrees=45
  swing2Motion=Free
  twistLimitDegrees=45
  twistMotion=Free
 }
 {
  bone1Name=thigh_r
  bone2Name=pelvis
  swing1LimitDegrees=45
  swing1Motion=Free
  swing2LimitDegrees=45
  swing2Motion=Free
  twistLimitDegrees=45
  twistMotion=Free
 }
 {
  bone1Name=base_jacket_b
  bone2Name=spine_01
  swing1LimitDegrees=45
  swing1Motion=Free
  swing2LimitDegrees=45
  swing2Motion=Free
  twistLimitDegrees=45
  twistMotion=Free
 }
 {bone1Name=dyn_jacket_b, bone2Name=base_jacket_b, swing1LimitDegrees=2, swing2LimitDegrees=15, twistLimitDegrees=1}
 {
  bone1Name=base_jacket_l
  bone2Name=spine_01
  swing1LimitDegrees=45
  swing1Motion=Free
  swing2LimitDegrees=45
  swing2Motion=Free
  twistLimitDegrees=45
  twistMotion=Free
 }
 {bone1Name=dyn_jacket_l, bone2Name=base_jacket_l, swing1LimitDegrees=20, swing2LimitDegrees=15, twistLimitDegrees=1}
 {
  bone1Name=base_jacket_r
  bone2Name=spine_01
  swing1LimitDegrees=45
  swing1Motion=Free
  swing2LimitDegrees=45
  swing2Motion=Free
  twistLimitDegrees=45
  twistMotion=Free
 }
 {bone1Name=dyn_jacket_r, bone2Name=base_jacket_r, swing1LimitDegrees=20, swing2LimitDegrees=15, twistLimitDegrees=1}
 {
  bone1Name=spine_01
  bone2Name=pelvis
  swing1LimitDegrees=45
  swing1Motion=Free
  swing2LimitDegrees=45
  swing2Motion=Free
  twistLimitDegrees=45
  twistMotion=Free
 }
 {bone1Name=dyn_f_jacket_l, bone2Name=base_f_jacket_l, swing1LimitDegrees=10, swing2LimitDegrees=10, twistLimitDegrees=3}
 {bone1Name=dyn_f_jacket_01_l, bone2Name=dyn_f_jacket_l, swing1LimitDegrees=10, swing2LimitDegrees=10, twistLimitDegrees=3}
 {bone1Name=dyn_f_jacket_02_l, bone2Name=dyn_f_jacket_01_l, swing1LimitDegrees=20, swing2LimitDegrees=10, twistLimitDegrees=3}
 {bone1Name=dyn_f_jacket_r, bone2Name=base_f_jacket_r, swing1LimitDegrees=10, swing2LimitDegrees=10, twistLimitDegrees=3}
 {bone1Name=dyn_f_jacket_01_r, bone2Name=dyn_f_jacket_r, swing1LimitDegrees=10, swing2LimitDegrees=10, twistLimitDegrees=3}
 {bone1Name=dyn_f_jacket_02_r, bone2Name=dyn_f_jacket_01_r, swing1LimitDegrees=20, swing2LimitDegrees=10, twistLimitDegrees=1}
 {
  bone1Name=dyn_knife
  bone2Name=pelvis
  swing1LimitDegrees=45
  swing1Motion=Free
  swing2LimitDegrees=45
  swing2Motion=Free
  twistLimitDegrees=45
  twistMotion=Free
 }
]
