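# Control Rig for the female medium SurgeRaven ZanyOwl player body, applying corrective deformation, mainly to jacket and knife bones.
# Construction adds FK controls; the forward solve chains FK setup, a pose-reader left forearm twist fix, shared DeformRig and JacketDeformation_FNC.
# The jacket function weights additive bone offsets by remapped thigh and knife rotations, thigh-up anim attributes, smoothed crouch variables and dbno.
# A zipper parent-constraint function exists but is unwired.
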ControlRigBlueprint /BRCosmetics/Characters/Player/Female/Medium/Bodies/F_MED_SurgeRaven_ZanyOwl/Meshes/M_MED_SurgeRaven_ZanyOwl_CtrlRig
graphs [
 New Function_ContainedGraph {nodes [
   RigVMFunctionEntryNode()
   > n1 = &RigUnit_SphericalPoseReader(
    ActiveRegionScaleFactors=(PositiveWidth=2,NegativeWidth=1,PositiveHeight=1,NegativeHeight=1)
    ActiveRegionSize=0.12
    Debug=(bDrawDebug=false,bDraw2D=false,DebugScale=25,DebugSegments=20,DebugThickness=0.25)
    DriverAxis=(X=1,Y=0,Z=0)
    DriverItem=(Type=Bone,Name="thumb_03_l")
    FalloffRegionScaleFactors=(PositiveWidth=-2,NegativeWidth=1,PositiveHeight=1,NegativeHeight=1)
    FalloffSize=0.2
    OptionalParentItem=(Type=Bone,Name="lowerarm_l")
    RotationOffset=(X=-30,Y=30,Z=20)
   )
   > &RigUnit_ModifyTransforms(
    ItemToModify="((Item=(Type=Bone,Name=\"lowerarm_twist_01_l\"),Transform=(Rotation=(X=0,Y=-0.0436,Z=0,W=0.999),Translation=(X=-0.7,Y=-0.17,Z=1),Scale3D=(X=1,Y=1,Z=1))),(Item=(Type=Bone,Name=\"lowerarm_twist_02_l\"),Transform=(Rotation=(X=0.1736,Y=0,Z=0,W=0.9848),Translation= ...(truncated)"
    WeightMinimum=0
    Weight=@n1.OutputParam
   )
   > RigVMFunctionReturnNode()
  ]}
 New Function_ContainedGraph {nodes [
   RigVMFunctionEntryNode()
   > RigVMFunction_Sequence_3 = RigVMAggregateNode()
   n1 = &RigVMFunction_Sequence(ExecuteContext=@RigVMFunction_Sequence_3.A)
   Delta Rotation Y_1 = RigVMFunctionReferenceNode(Child=thigh_r, Parent=thigh_r_null, Source Maximum=-80, ExecuteContext=@n1.B)
   > &RigUnit_ModifyTransforms(
    ItemToModify="((Item=(Type=Bone,Name=\"base_f_jacket_r\"),Transform=(Rotation=(X=0,Y=-0.1736,Z=0,W=0.9848),Translation=(X=0,Y=0,Z=0),Scale3D=(X=1,Y=1,Z=1))),(Item=(Type=Bone,Name=\"base_jacket_r\"),Transform=(Rotation=(X=0,Y=0,Z=-0.2588,W=0.9659),Translation=(X=0,Y=0,Z=0),S ...(truncated)"
    WeightMinimum=0
    Weight=@"Delta Rotation Y_1".Delta Rotation
   )
   > Delta Rotation Y_4 = RigVMFunctionReferenceNode(Child=thigh_r, Parent=thigh_r_null, Source Maximum=-110, Source Minimum=0)
   > &RigUnit_ModifyTransforms(
    ItemToModify=((Item=(Type=Bone,Name="base_jacket_r"),Transform=(Rotation=(X=0,Y=0,Z=-0.5,W=0.866),Translation=(X=-5,Y=0,Z=0),Scale3D=(X=1,Y=1,Z=1))))
    WeightMinimum=0
    Weight=@"Delta Rotation Y_4".Delta Rotation
   )
   &RigUnit_ModifyTransforms(
    ItemToModify=((Item=(Type=Bone,Name="base_f_jacket_r"),Transform=(Rotation=(X=-0.3172,Y=-0.1714,Z=0.0936,W=0.928),Translation=(X=0,Y=0,Z=0),Scale3D=(X=1,Y=1,Z=1))))
    WeightMinimum=0
    Weight=RigVMDispatchNode(BoneName=root, Default=1, Name=deformRig_r_thigh_up_alpha) {template="DISPATCH_RigDispatch_GetAnimAttribute(in Name,in BoneName,in Default,out Value,out Found)"}.Value
    ExecutePin=@n1.A
   )
   > Delta Rotation Z_2 = RigVMFunctionReferenceNode(Child=thigh_r, Parent=thigh_r_null, Source Maximum=-140, Source Minimum=-80)
   > &RigUnit_ModifyTransforms(
    ItemToModify=((Item=(Type=Bone,Name="base_f_jacket_r"),Transform=(Rotation=(X=0,Y=-0.342,Z=0,W=0.9397),Translation=(X=0,Y=0,Z=0),Scale3D=(X=1,Y=1,Z=1))))
    WeightMinimum=0
    Weight=@"Delta Rotation Z_2".Delta Rotation
   )
   > Delta Rotation Z_5 = RigVMFunctionReferenceNode(Child=thigh_r, Parent=thigh_r_null, Source Maximum=-140, Source Minimum=-30)
   > &RigUnit_ModifyTransforms(
    ItemToModify=((Item=(Type=Bone,Name="base_jacket_r"),Transform=(Rotation=(X=0,Y=0,Z=-0.5,W=0.866),Translation=(X=0,Y=0,Z=0),Scale3D=(X=1,Y=1,Z=1))))
    WeightMinimum=0
    Weight=@"Delta Rotation Z_5".Delta Rotation
   )
   RigVMFunctionReturnNode(ExecuteContext=@RigVMFunction_Sequence_3.D)
   RigVMFunction_Sequence = &RigVMFunction_Sequence(ExecuteContext=@RigVMFunction_Sequence_3.C)
   &RigUnit_ModifyTransforms(
    ItemToModify=((Item=(Type=Bone,Name="base_knife"),Transform=(Rotation=(X=0.0269,Y=0.6069,Z=-0.215,W=0.7647),Translation=(X=-1,Y=-1,Z=0),Scale3D=(X=1,Y=1,Z=1))))
    WeightMinimum=0
    Weight=RigVMDispatchNode(BoneName=root, Default=1, Name=deformRig_r_thigh_up_alpha) {template="DISPATCH_RigDispatch_GetAnimAttribute(in Name,in BoneName,in Default,out Value,out Found)"}.Value
    ExecutePin=@RigVMFunction_Sequence.A
   )
   > Delta Rotation Z_1 = RigVMFunctionReferenceNode(Child=thigh_r, Parent=thigh_r_null, Source Maximum=-110, Source Minimum=-80)
   > &RigUnit_ModifyTransforms(
    ItemToModify=((Item=(Type=Bone,Name="base_knife"),Transform=(Rotation=(X=0.0625,Y=0.3377,Z=-0.0937,W=0.9345),Translation=(X=0,Y=0,Z=0),Scale3D=(X=1,Y=1,Z=1))))
    WeightMinimum=0
    Weight=@"Delta Rotation Z_1".Delta Rotation
   )
   > Delta Rotation Z_3 = RigVMFunctionReferenceNode(Child=base_knife, Parent=base_knife_null, Source Maximum=-20, Source Minimum=0)
   > &RigUnit_ModifyTransforms(
    ItemToModify="((Item=(Type=Bone,Name=\"base_jacket_r\"),Transform=(Rotation=(X=-0.117,Y=-0.3214,Z=-0.3214,W=0.883),Translation=(X=0,Y=0,Z=0),Scale3D=(X=1,Y=1,Z=1))),(Item=(Type=Bone,Name=\"base_f_jacket_r\"),Transform=(Rotation=(X=0.0019,Y=-0.0436,Z=0.0436,W=0.9981),Transla ...(truncated)"
    WeightMinimum=0
    Weight=@"Delta Rotation Z_3".Delta Rotation
   )
   > Delta Rotation Z_4 = RigVMFunctionReferenceNode(Child=base_knife, Parent=base_knife_null, Source Maximum=-20, Source Minimum=0)
   > &RigUnit_ModifyTransforms(
    ItemToModify="((Item=(Type=Bone,Name=\"base_jacket_r\"),Transform=(Rotation=(X=0,Y=0,Z=0.7071,W=0.7071),Translation=(X=0,Y=0,Z=0),Scale3D=(X=1,Y=1,Z=1))),(Item=(Type=Bone,Name=\"base_f_jacket_r\"),Transform=(Rotation=(X=0,Y=0,Z=-0.0436,W=0.999),Translation=(X=0,Y=0,Z=0),Sca ...(truncated)"
    WeightMinimum=0
    Weight=@"Delta Rotation Z_4".Delta Rotation
   )
   Delta Rotation Y_3 = RigVMFunctionReferenceNode(Child=thigh_r, Parent=thigh_r_null, Source Maximum=-80, ExecuteContext=@RigVMFunction_Sequence.B)
   > &RigUnit_ModifyTransforms(
    ItemToModify=((Item=(Type=Bone,Name="base_knife"),Transform=(Rotation=(X=-0.1749,Y=-0.1908,Z=-0.6526,W=0.7122),Translation=(X=3,Y=-3,Z=0),Scale3D=(X=1,Y=1,Z=1))))
    WeightMinimum=0
    Weight=@"Delta Rotation Y_3".Delta Rotation
   )
   > Delta Rotation Y_2 = RigVMFunctionReferenceNode(Child=thigh_r, Parent=thigh_r_null, Source Maximum=20, Source Minimum=0)
   > &RigUnit_ModifyTransforms(
    ItemToModify="((Item=(Type=Bone,Name=\"base_knife\"),Transform=(Rotation=(X=0,Y=0,Z=0.1736,W=0.9848),Translation=(X=2,Y=-1,Z=0),Scale3D=(X=1,Y=1,Z=1))),(Item=(Type=Bone,Name=\"base_jacket_r\"),Transform=(Rotation=(X=0,Y=0,Z=-0.1736,W=0.9848),Translation=(X=0,Y=0,Z=0),Scale3 ...(truncated)"
    WeightMinimum=0
    Weight=@"Delta Rotation Y_2".Delta Rotation
   )
   > Delta Rotation Y_6 = RigVMFunctionReferenceNode(Child=thigh_r, Parent=thigh_r_null, Source Maximum=-110, Source Minimum=-50)
   > &RigUnit_ModifyTransforms(
    ItemToModify=((Item=(Type=Bone,Name="base_knife"),Transform=(Rotation=(X=0,Y=0,Z=-0.9397,W=0.342),Translation=(X=-2,Y=0,Z=0),Scale3D=(X=1,Y=1,Z=1))))
    WeightMinimum=0
    Weight=@"Delta Rotation Y_6".Delta Rotation
   )
   n2 = &RigVMFunction_Sequence(ExecuteContext=@RigVMFunction_Sequence_3.B)
   Delta Rotation Y = RigVMFunctionReferenceNode(Child=thigh_l, Parent=thigh_l_null, Source Maximum=-80, Source Minimum=0, ExecuteContext=@n2.B)
   > &RigUnit_ModifyTransforms(
    ItemToModify="((Item=(Type=Bone,Name=\"base_f_jacket_l\"),Transform=(Rotation=(X=-0.0302,Y=0.171,Z=0.171,W=0.9698),Translation=(X=0,Y=0,Z=0),Scale3D=(X=1,Y=1,Z=1))),(Item=(Type=Bone,Name=\"base_jacket_l\"),Transform=(Rotation=(X=0,Y=0,Z=0.1736,W=0.9848),Translation=(X=0,Y=0 ...(truncated)"
    WeightMinimum=0
    Weight=@"Delta Rotation Y".Delta Rotation
   )
   > Delta Rotation Y_5 = RigVMFunctionReferenceNode(Child=thigh_l, Parent=thigh_l_null, Source Maximum=-70, Source Minimum=-30)
   > &RigUnit_ModifyTransforms(
    ItemToModify="((Item=(Type=Bone,Name=\"base_jacket_l\"),Transform=(Rotation=(X=0.5,Y=-0.5,Z=0.5,W=0.5),Translation=(X=0,Y=0,Z=0),Scale3D=(X=1,Y=1,Z=1))),(Item=(Type=Bone,Name=\"base_f_jacket_l\"),Transform=(Rotation=(X=-0.1294,Y=0.2241,Z=0.483,W=0.8365),Translation=(X=0,Y=0 ...(truncated)"
    WeightMinimum=0
    Weight=@"Delta Rotation Y_5".Delta Rotation
   )
   &RigUnit_ModifyTransforms(
    ItemToModify=((Item=(Type=Bone,Name="base_f_jacket_l"),Transform=(Rotation=(X=0.2831,Y=0.2969,Z=0.0704,W=0.9093),Translation=(X=0,Y=2,Z=-2),Scale3D=(X=1,Y=1,Z=1))))
    WeightMinimum=0
    Weight=RigVMDispatchNode(BoneName=root, Default=1, Name=deformRig_l_thigh_up_alpha) {template="DISPATCH_RigDispatch_GetAnimAttribute(in Name,in BoneName,in Default,out Value,out Found)"}.Value
    ExecutePin=@n2.A
   )
   > Delta Rotation Z = RigVMFunctionReferenceNode(Child=thigh_l, Parent=thigh_l_null, Source Maximum=-120, Source Minimum=-80)
   > &RigUnit_ModifyTransforms(
    ItemToModify=((Item=(Type=Bone,Name="base_f_jacket_l"),Transform=(Rotation=(X=0,Y=0.342,Z=0,W=0.9397),Translation=(X=0,Y=0,Z=0),Scale3D=(X=1,Y=1,Z=1))))
    WeightMinimum=0
    Weight=@"Delta Rotation Z".Delta Rotation
   )
   > Delta Rotation Z_6 = RigVMFunctionReferenceNode(Child=thigh_l, Parent=thigh_l_null, Source Maximum=-140, Source Minimum=-30)
   > &RigUnit_ModifyTransforms(
    ItemToModify=((Item=(Type=Bone,Name="base_jacket_l"),Transform=(Rotation=(X=0,Y=0,Z=0.5,W=0.866),Translation=(X=0,Y=0,Z=0),Scale3D=(X=1,Y=1,Z=1))))
    WeightMinimum=0
    Weight=@"Delta Rotation Z_6".Delta Rotation
   )
   n3 = &RigVMFunction_AlphaInterp(
    InterpSpeedDecreasing=10
    InterpSpeedIncreasing=5
    Scale=1
    bClampResult=true
    bMapRange=true
    Value=RigVMVariableNode() {variable=Crouch}.Value
   )
   > &RigUnit_ModifyTransforms(
    ItemToModify="((Item=(Type=Bone,Name=\"base_f_jacket_l\"),Transform=(Rotation=(X=0,Y=0,Z=0.342,W=0.9397),Translation=(X=0,Y=0,Z=0),Scale3D=(X=1,Y=1,Z=1))),(Item=(Type=Bone,Name=\"base_f_jacket_r\"),Transform=(Rotation=(X=0.25,Y=0.25,Z=-0.067,W=0.933),Translation=(X=2,Y=0,Z= ...(truncated)"
    WeightMinimum=0
    Weight=@n3.Result
   )
   > &RigUnit_ModifyTransforms(
    ItemToModify="((Item=(Type=Bone,Name=\"base_f_jacket_l\"),Transform=(Rotation=(X=0.2588,Y=0,Z=0,W=0.9659),Translation=(X=0,Y=0,Z=0),Scale3D=(X=1,Y=1,Z=1))),(Item=(Type=Bone,Name=\"base_f_jacket_r\"),Transform=(Rotation=(X=0.1574,Y=-0.0734,Z=-0.4162,W=0.8925),Translation=(X= ...(truncated)"
    WeightMinimum=0
    Weight=RigVMVariableNode() {variable=dbno}.Value
   )
   n4 = &RigVMFunction_AlphaInterp(
    InterpSpeedDecreasing=10
    InterpSpeedIncreasing=10
    Scale=1
    bClampResult=true
    bMapRange=true
    Value=RigVMVariableNode() {variable=CrouchMoving}.Value
   )
   > &RigUnit_ModifyTransforms(
    ItemToModify=((Item=(Type=Bone,Name="base_jacket_b"),Transform=(Rotation=(X=0,Y=0.1736,Z=0,W=0.9848),Translation=(X=2,Y=0,Z=0),Scale3D=(X=1,Y=1,Z=1))))
    WeightMinimum=0
    Weight=@n4.Result
   )
  ]}
 New Function_ContainedGraph {nodes [
   n1 = RigVMFunctionEntryNode()
   RigVMFunction_MathQuaternionToEuler = &RigVMFunction_MathQuaternionToEuler(
    RotationOrder=XYZ
    Value=&RigUnit_GetRelativeTransformForItem(
     Child=(Type=Bone,Name="hand_l")
     Parent=(Type=Null,Name="lowerarm_l")
     bChildInitial=False
     bParentInitial=False
     Child.Name=@n1.Child
     Parent.Name=@n1.Parent
    ).RelativeTransform.Rotation
   )
   Remap = &RigVMFunction_MathDoubleRemap(
    TargetMaximum=1
    TargetMinimum=0
    Value=@RigVMFunction_MathQuaternionToEuler.Result.Y
    SourceMaximum=@n1.Source Maximum
    SourceMinimum=@n1.Source Minimum
   )
   > RigVMFunctionReturnNode(Delta Rotation=@Remap.Result)
  ]}
 New Function_ContainedGraph {nodes [
   n1 = RigVMFunctionEntryNode()
   RigVMFunction_MathQuaternionToEuler = &RigVMFunction_MathQuaternionToEuler(
    RotationOrder=XYZ
    Value=&RigUnit_GetRelativeTransformForItem(
     Child=(Type=Bone,Name="hand_l")
     Parent=(Type=Null,Name="lowerarm_l")
     bChildInitial=False
     bParentInitial=False
     Child.Name=@n1.Child
     Parent.Name=@n1.Parent
    ).RelativeTransform.Rotation
   )
   Remap = &RigVMFunction_MathDoubleRemap(
    TargetMaximum=1
    TargetMinimum=0
    Value=@RigVMFunction_MathQuaternionToEuler.Result.Z
    SourceMaximum=@n1.Source Maximum
    SourceMinimum=@n1.Source Minimum
   )
   > RigVMFunctionReturnNode(Delta Rotation=@Remap.Result)
  ]}
 New Function_ContainedGraph {nodes [
   RigVMFunctionEntryNode()
   > &RigUnit_ParentConstraint(
    AdvancedSettings=(RotationOrderForFilter=XZY)
    Child=(Type=Bone,Name="base_m_zipper_r")
    Parents="((Item=(Type=Bone,Name=\"spine_01\"),Weight=0.0013),(Item=(Type=Bone,Name=\"spine_03\"),Weight=0.0034),(Item=(Type=Bone,Name=\"spine_05\"),Weight=0.004),(Item=(Type=Bone,Name=\"spine_02\"),Weight=0.0042),(Item=(Type=Bone,Name=\"dyn_f_jacket_02_r\"),Weight=0.0052),(I ...(truncated)"
    Weight=1
   )
   > &RigUnit_ParentConstraint(
    AdvancedSettings=(RotationOrderForFilter=XZY)
    Child=(Type=Bone,Name="base_lower_zipper_r")
    Parents="((Item=(Type=Bone,Name=\"spine_02\"),Weight=0.0001),(Item=(Type=Bone,Name=\"deform_elbow_in_base_r\"),Weight=0.0002),(Item=(Type=Bone,Name=\"pelvis\"),Weight=0.0006),(Item=(Type=Bone,Name=\"dyn_jacket_r\"),Weight=0.0024),(Item=(Type=Bone,Name=\"dyn_f_jacket_r\"),Wei ...(truncated)"
    Weight=1
   )
   > &RigUnit_ParentConstraint(
    AdvancedSettings=(RotationOrderForFilter=XZY)
    Child=(Type=Bone,Name="base_t_zipper_l")
    Parents=((Item=(Type=Bone,Name="clavicle_l"),Weight=0.0169),(Item=(Type=Bone,Name="spine_05"),Weight=0.1294),(Item=(Type=Bone,Name="spine_04"),Weight=0.8537))
    Weight=1
   )
   > &RigUnit_ParentConstraint(
    AdvancedSettings=(RotationOrderForFilter=XZY)
    Child=(Type=Bone,Name="base_t_zipper_r")
    Parents=((Item=(Type=Bone,Name="clavicle_r"),Weight=0.0166),(Item=(Type=Bone,Name="spine_05"),Weight=0.1317),(Item=(Type=Bone,Name="spine_04"),Weight=0.8517))
    Weight=1
   )
   > RigVMFunctionReturnNode()
  ]}
 New Function_ContainedGraph {nodes [
   n1 = RigVMFunctionEntryNode()
   RigVMFunction_MathQuaternionToEuler = &RigVMFunction_MathQuaternionToEuler(
    RotationOrder=XYZ
    Value=&RigUnit_GetRelativeTransformForItem(
     Child=(Type=Bone,Name="hand_l")
     Parent=(Type=Null,Name="lowerarm_l")
     bChildInitial=False
     bParentInitial=False
     Child.Name=@n1.Child
     Parent.Name=@n1.Parent
    ).RelativeTransform.Rotation
   )
   Remap = &RigVMFunction_MathDoubleRemap(
    TargetMaximum=1
    TargetMinimum=0
    Value=@RigVMFunction_MathQuaternionToEuler.Result.X
    SourceMaximum=@n1.Source Maximum
    SourceMinimum=@n1.Source Minimum
   )
   > RigVMFunctionReturnNode(Delta Rotation=@Remap.Result)
  ]}
 RigVMFunctionLibrary {nodes [
   Constraints = RigVMCollapseNode(ExecuteContext=())
   Delta Rotation X = RigVMCollapseNode(ExecuteContext=())
   Delta Rotation Y = RigVMCollapseNode(ExecuteContext=())
   Delta Rotation Z = RigVMCollapseNode(ExecuteContext=())
   JacketDeformation_FNC = RigVMCollapseNode(ExecuteContext=())
  ]}
 RigVMFunction_Sequence_4_ContainedGraph {nodes [
   n1 = RigVMFunctionEntryNode()
   n2 = &RigVMFunction_Sequence(ExecuteContext=@n1.ExecuteContext)
   n3 = &RigVMFunction_Sequence(ExecuteContext=@n2.B)
   n4 = &RigVMFunction_Sequence(ExecuteContext=@n3.B)
   RigVMFunctionReturnNode(B=@n3.A, C=@n4.A, D=@n4.B, A=@n2.A)
  ]}
 RigVMModel {nodes [
   n1 = &RigUnit_PrepareForExecution()
   addFKCtrl = RigVMFunctionReferenceNode(
    Custom_FKCtrl="((joint_names_14_#1=((Type=Bone,Name=\"base_knife\")),parent_2_#2=(Type=Control,Name=\"base_knife_ctrl\"),shape_settings_24_#3=(bVisible=True,Name=\"Hexagon_Thick\",Color=( ...(truncated)"
    Options_FKCtrl="(pelvis_1_#4=False,spine_3_#5=False,clavicle_5_#6=False,arm_7_#7=False,leg_9_#8=true,head_11_46DC15EF465CD ...(truncated)"
    Shape_Settings="(pelvis_2_#9=(bVisible=True,Name=\"Hexagon_Thin\",Color=(R=0.0117,G=0.8281,B=1,A=1),Transform=(Rotation=(X=0,Y=-0.7071,Z=0,W=0.7071),Translation=(X=2,Y=0,Z=0),Scale3D=(X=5,Y=5,Z=7))),spine_4_#10=(bVi ...(truncated)"
    ExecuteContext=@n1.ExecutePin
   )
   RigUnit_BeginExecution = &RigUnit_BeginExecution()
   setFKCtrl = RigVMFunctionReferenceNode(setFK_Ctrl=false, ExecuteContext=@RigUnit_BeginExecution.ExecutePin)
   > TwistsFix = RigVMCollapseNode()
   > FN_DeformRig_FNC = RigVMFunctionReferenceNode(
    DeformRig_ Curve="(deformRig_l_upperarm_up_alpha_72_#11=False,deformRig_r_upperarm_up_alpha_73_#12=False,deformRig_l_upperarm_fr_alpha_74_#13=False,deformRig_r_upperarm_fr_alpha_75_CB51B8 ...(truncated)"
    DeformRig_Bool="(upperarm_1_#14=True,elbow_3_#15=true,wrist_7_#16=True,groin_8_#17=true,glute_17_#18=true,knee_9_40C408324CA261 ...(truncated)"
    DeformRig_Debug=(upperarm_debug_18_#14=False,wrist_debug_20_#16=False,lat_debug_26_#19=False,groinGlute_debug_27_#17=False)
    DeformRig_Value="(upperarm_up_18_#20=(Rotation=(X=0,Y=0,Z=0,W=1),Translation=(X=0,Y=0,Z=0),Scale3D=(X=1,Y=1,Z=1)),upperarm_fwd_fr_19_#21=(Rotation=(X=0,Y=0,Z=0,W=1),Translation=(X=0,Y=10,Z=0),Scale3D=(X=0.5,Y=1,Z=1) ...(truncated)"
   )
   > JacketDeformation_FNC = RigVMFunctionReferenceNode()
   Constraints = RigVMFunctionReferenceNode(ExecuteContext=())
  ]}
]
hierarchy [
 root {loc=<0 0 0>, rot=<0 0 0 1>}
 pelvis {parent=root, loc=<0 -0.5301 98.0571>, rot=<0.0127 0.707 -0.0127 -0.707>}
 spine_01 {parent=pelvis, loc=<6.1208 -0.9328 0>, rot=<0 0 0.034 -0.9994>}
 spine_02 {parent=spine_01, loc=<6.5997 -0.2095 0>, rot=<0 0 0.0238 -0.9997>}
 spine_03 {parent=spine_02, loc=<6.4914 -0.1237 0>, rot=<0 0 0.0149 -0.9999>}
 spine_04 {parent=spine_03, loc=<6.9598 0.0794 0>, rot=<0 0 -0.0132 -0.9999>}
 spine_05 {parent=spine_04, loc=<6.9987 0 0>, rot=<0 0 -0.0015 -1>}
 neck_01 {parent=spine_05, loc=<11.3138 0.1529 0>, rot=<0 0 -0.2003 -0.9797>}
 neck_02 {parent=neck_01, loc=<4.5627 0 0>, rot=<0 0 0.0504 -0.9987>}
 head {parent=neck_02, loc=<5.3914 0 0>, rot=<0 0 0.115 -0.9934>}
 base_bangs {parent=head, loc=<9.7244 7.3958 0.5868>, rot=<0 0 -0.7071 -0.7071>}
 dyn_bangs {parent=base_bangs, loc=<0 0 0>, rot=<0 0 0 1>}
 base_bangs_sml_r {parent=head, loc=<4.2093 5.5696 5.6596>, rot=<0 0 0.999 -0.0436>}
 dyn_bangs_sml_r {parent=base_bangs_sml_r, loc=<0 0 0>, rot=<0 0 0 1>}
 base_bangs_lrg_r {parent=head, loc=<5.0811 3.8136 6.6609>, rot=<0 0 0.999 -0.0436>}
 dyn_bangs_lrg_r {parent=base_bangs_lrg_r, loc=<0 0 0>, rot=<0 0 0 1>}
 dyn_bangs_lrg_01_r {parent=dyn_bangs_lrg_r, loc=<4.4071 -0.3489 -0.1493>, rot=<0 0 0.0436 -0.999>}
 base_bangs_l {parent=head, loc=<4.6532 3.8914 -5.9866>, rot=<0 0 0.999 -0.0436>}
 dyn_bangs_l {parent=base_bangs_l, loc=<0 0 0>, rot=<0 0 0 1>}
 dyn_bangs_01_l {parent=dyn_bangs_l, loc=<2.7189 0.1685 0>, rot=<0 0 0 1>}
 base_ponytail {parent=head, loc=<9.1051 -6.7455 0.0418>, rot=<0 0 0.6428 -0.766>}
 dyn_ponytail {parent=base_ponytail, loc=<0 0 0>, rot=<0 0 0 1>}
 dyn_ponytail_01 {parent=dyn_ponytail, loc=<4.168 -0.0283 0>, rot=<0 0 0.766 -0.6428>}
 dyn_ponytail_02 {parent=dyn_ponytail_01, loc=<7.1113 -0.1258 0>, rot=<0 0 0.0436 -0.999>}
 dyn_ponytail_end_b {parent=dyn_ponytail_02, loc=<7.7433 2.4748 0>, rot=<0 0 0 1>}
 dyn_ponytail_end_f {parent=dyn_ponytail_02, loc=<4.0046 -3.0836 0>, rot=<0 0 -0.0872 -0.9962>}
 dyn_ponytail_end_l {parent=dyn_ponytail_02, loc=<3.6658 -0.4962 -2.9056>, rot=<0 0 0 1>}
 dyn_ponytail_end_r {parent=dyn_ponytail_02, loc=<3.5337 0.8819 2.5712>, rot=<0 0 0 1>}
 dyn_ponytail_end_m {parent=dyn_ponytail_02, loc=<5.124 -0.3564 0>, rot=<0 0 0 1>}
 dyn_ponytail_end_ext {parent=dyn_ponytail_end_m, loc=<6.0445 1.2881 0>, rot=<0 0 -0.0872 -0.9962>}
 attach_cape {parent=spine_05, loc=<9.017 -11.714 0>, rot=<0.0616 -0.7044 -0.0616 -0.7044>}
 cape {parent=attach_cape, type=Null, loc=<0 0 0>, rot=<0 0 0 1>}
 attach_backpack {parent=spine_05, loc=<-0.123 -13.473 0>, rot=<-0.0093 -0.707 0.0093 -0.707>}
 Backpack_BR {parent=attach_backpack, type=Null, loc=<0 0 0>, rot=<0 0 0 1>}
 clavicle_l {parent=spine_05, loc=<5.4831 -0.1053 -3.2946>, rot=<0.0465 -0.7289 -0.009 -0.6829>}
 upperarm_l {parent=clavicle_l, loc=<9.8658 -1.2136 -0.4105>, rot=<0.0164 -0.3245 -0.0269 -0.9454>}
 lowerarm_l {parent=upperarm_l, loc=<26.465 -0.8345 0.259>, rot=<0.0036 0.0161 -0.2298 -0.9731>}
 hand_l {parent=lowerarm_l, loc=<22.7355 -0.3203 -0.6105>, rot=<-0.6784 -0.0223 0.0735 -0.7307>}
 thumb_01_l {parent=hand_l, loc=<1.3118 -0.3191 -2.459>, rot=<0.6546 -0.4042 0.0664 -0.6354>}
 thumb_02_l {parent=thumb_01_l, loc=<5.9336 -0.0448 0.0401>, rot=<-0.0037 0.0213 0.298 -0.9543>}
 thumb_03_l {parent=thumb_02_l, loc=<2.2223 0.3682 -0.0907>, rot=<0.0118 -0.013 0.0066 -0.9998>}
 index_metacarpal_l {parent=hand_l, loc=<1.6535 0.0183 -2.96>, rot=<0.0158 -0.0124 -0.0173 -0.9996>}
 index_01_l {parent=index_metacarpal_l, loc=<8.3008 0.0357 -0.0524>, rot=<-0.0044 0.0485 0.0924 -0.9945>}
 index_02_l {parent=index_01_l, loc=<3.3587 -0.3004 0.0452>, rot=<-0.002 0.0027 0.3225 -0.9466>}
 index_03_l {parent=index_02_l, loc=<2.5578 -0.123 -0.0252>, rot=<-0.002 0.0067 0.2558 -0.9667>}
 middle_metacarpal_l {parent=hand_l, loc=<2.1856 -0.1725 -0.5984>, rot=<0.006 0.0168 -0.0329 -0.9993>}
 middle_01_l {parent=middle_metacarpal_l, loc=<7.3275 -0.2447 0.0129>, rot=<-0.0208 0.0042 0.0858 -0.9961>}
 middle_02_l {parent=middle_01_l, loc=<4.0123 0.1401 0.1175>, rot=<-0.0218 0.0258 0.3512 -0.9357>}
 middle_03_l {parent=middle_02_l, loc=<2.6077 0.0534 0.1373>, rot=<0.0082 0.0188 0.1708 -0.9851>}
 ring_metacarpal_l {parent=hand_l, loc=<2.2074 -0.3625 1.3638>, rot=<-0.0284 0.0722 -0.0458 -0.9959>}
 ring_01_l {parent=ring_metacarpal_l, loc=<7.0295 -0.6197 -0.1446>, rot=<-0.0489 -0.0252 0.1892 -0.9804>}
 ring_02_l {parent=ring_01_l, loc=<3.6174 0.4647 -0.1027>, rot=<0 0 0.2806 -0.9598>}
 ring_03_l {parent=ring_02_l, loc=<2.4152 0.114 0.0199>, rot=<0 0 0.1189 -0.9929>}
 pinky_metacarpal_l {parent=hand_l, loc=<1.7616 -0.9753 3.3814>, rot=<-0.0397 0.0814 -0.069 -0.9935>}
 pinky_01_l {parent=pinky_metacarpal_l, loc=<7.1709 -0.4831 -0.1094>, rot=<-0.0654 -0.0035 0.2242 -0.9723>}
 pinky_02_l {parent=pinky_01_l, loc=<2.8274 0.0893 -0.1043>, rot=<0.0037 0.0011 0.2854 -0.9584>}
 pinky_03_l {parent=pinky_02_l, loc=<2.2046 0.0307 -0.0078>, rot=<-0.0002 0.0098 0.1445 -0.9895>}
 deform_wrist_dn_base_l {parent=hand_l, loc=<0 -1.5326 0>, rot=<0 0 0 1>}
 deform_wrist_dn_l {parent=deform_wrist_dn_base_l, loc=<0 0 0>, rot=<0 0 0 1>}
 deform_wrist_up_base_l {parent=hand_l, loc=<0 2.3009 0>, rot=<0 0 0 1>}
 deform_wrist_up_l {parent=deform_wrist_up_base_l, loc=<0 0 0>, rot=<0 0 0 1>}
 hand_attach_l {parent=hand_l, loc=<8.5 -3 0>, rot=<0.7071 0.7071 0 0>}
 weapon_l {parent=hand_l, loc=<-0.1974 -3.4111 0.3811>, rot=<0 0 0 1>}
 lowerarm_twist_02_l {parent=lowerarm_l, loc=<10.6707 0 0>, rot=<0 0 0 1>}
 lowerarm_twist_01_l {parent=lowerarm_l, loc=<18.316 0 0>, rot=<0 0 0 1>}
 base_l_cuff_t {parent=lowerarm_twist_01_l, loc=<1.5134 -2.2359 2.7124>, rot=<0 0 -0.7373 -0.6756>}
 dyn_l_cuff_t {parent=base_l_cuff_t, loc=<0 0 0>, rot=<0 0 0 1>}
 base_l_cuff_b {parent=lowerarm_twist_01_l, loc=<2.6708 -2.3396 -4.4935>, rot=<0 0 -0.7373 -0.6756>}
 dyn_l_cuff_b {parent=base_l_cuff_b, loc=<0 0 0>, rot=<0 0 0 1>}
 deform_elbow_in_base_l {parent=lowerarm_l, loc=<0.1834 3.6062 -0.2572>, rot=<0 0 0 1>}
 deform_elbow_in_l {parent=deform_elbow_in_base_l, loc=<0 0 0>, rot=<0 0 0 1>}
 deform_elbow_out_base_l {parent=lowerarm_l, loc=<0 -4 0>, rot=<0 0 0 -1>}
 deform_elbow_out_l {parent=deform_elbow_out_base_l, loc=<0 0 0>, rot=<0 0 0 1>}
 upperarm_twist_01_l {parent=upperarm_l, loc=<7.2135 -0.1134 0>, rot=<0 0 0 1>}
 deform_upperarm_bk_base_l {parent=upperarm_twist_01_l, loc=<-6.4053 -4.3746 0.314>, rot=<0 0 0 1>}
 deform_upperarm_bk_l {parent=deform_upperarm_bk_base_l, loc=<0 0 0>, rot=<0 0 0 1>}
 deform_upperarm_fr_base_l {parent=upperarm_twist_01_l, loc=<-5.6312 5.092 0.3354>, rot=<0 0 0 1>}
 deform_upperarm_fr_l {parent=deform_upperarm_fr_base_l, loc=<0 0 0>, rot=<0 0 0 1>}
 deform_upperarm_up_base_l {parent=upperarm_twist_01_l, loc=<-8.891 0.4825 3.9517>, rot=<0 0 0 1>}
 deform_upperarm_up_l {parent=deform_upperarm_up_base_l, loc=<0 0 0>, rot=<0 0 0 1>}
 upperarm_twist_02_l {parent=upperarm_l, loc=<19.7998 -0.1134 0>, rot=<0 0 0 1>}
 clavicle_r {parent=spine_05, loc=<5.4831 -0.1053 3.2946>, rot=<0.7289 0.0465 -0.6829 0.009>}
 upperarm_r {parent=clavicle_r, loc=<-9.8658 1.2136 0.4105>, rot=<0.0164 -0.3245 -0.0269 -0.9454>}
 lowerarm_r {parent=upperarm_r, loc=<-26.465 0.8345 -0.259>, rot=<0.0036 0.0161 -0.2298 -0.9731>}
 hand_r {parent=lowerarm_r, loc=<-22.7355 0.3203 0.6106>, rot=<-0.6784 -0.0223 0.0735 -0.7307>}
 thumb_01_r {parent=hand_r, loc=<-1.3118 0.3191 2.459>, rot=<0.6546 -0.4042 0.0664 -0.6354>}
 thumb_02_r {parent=thumb_01_r, loc=<-5.9336 0.0448 -0.0401>, rot=<-0.0037 0.0213 0.298 -0.9543>}
 thumb_03_r {parent=thumb_02_r, loc=<-2.2223 -0.3682 0.0907>, rot=<0.0118 -0.013 0.0066 -0.9998>}
 index_metacarpal_r {parent=hand_r, loc=<-1.6535 -0.0183 2.96>, rot=<0.0158 -0.0124 -0.0173 -0.9996>}
 index_01_r {parent=index_metacarpal_r, loc=<-8.3008 -0.0357 0.0524>, rot=<-0.0044 0.0485 0.0924 -0.9945>}
 index_02_r {parent=index_01_r, loc=<-3.3587 0.3004 -0.0452>, rot=<-0.002 0.0027 0.3225 -0.9466>}
 index_03_r {parent=index_02_r, loc=<-2.5578 0.123 0.0252>, rot=<-0.002 0.0067 0.2558 -0.9667>}
 middle_metacarpal_r {parent=hand_r, loc=<-2.1856 0.1725 0.5984>, rot=<0.006 0.0168 -0.0329 -0.9993>}
 middle_01_r {parent=middle_metacarpal_r, loc=<-7.3275 0.2447 -0.0129>, rot=<-0.0208 0.0042 0.0858 -0.9961>}
 middle_02_r {parent=middle_01_r, loc=<-4.0123 -0.1401 -0.1175>, rot=<-0.0218 0.0258 0.3512 -0.9357>}
 middle_03_r {parent=middle_02_r, loc=<-2.6077 -0.0534 -0.1373>, rot=<0.0082 0.0188 0.1708 -0.9851>}
 ring_metacarpal_r {parent=hand_r, loc=<-2.2074 0.3625 -1.3638>, rot=<-0.0284 0.0722 -0.0458 -0.9959>}
 ring_01_r {parent=ring_metacarpal_r, loc=<-7.0295 0.6197 0.1446>, rot=<-0.0489 -0.0252 0.1892 -0.9804>}
 ring_02_r {parent=ring_01_r, loc=<-3.6174 -0.4647 0.1027>, rot=<0 0 0.2806 -0.9598>}
 ring_03_r {parent=ring_02_r, loc=<-2.4152 -0.114 -0.0199>, rot=<0 0 0.1189 -0.9929>}
 pinky_metacarpal_r {parent=hand_r, loc=<-1.7616 0.9753 -3.3814>, rot=<-0.0397 0.0814 -0.069 -0.9935>}
 pinky_01_r {parent=pinky_metacarpal_r, loc=<-7.1709 0.4831 0.1094>, rot=<-0.0654 -0.0035 0.2242 -0.9723>}
 pinky_02_r {parent=pinky_01_r, loc=<-2.8274 -0.0893 0.1043>, rot=<0.0037 0.0011 0.2854 -0.9584>}
 pinky_03_r {parent=pinky_02_r, loc=<-2.2046 -0.0307 0.0078>, rot=<-0.0002 0.0098 0.1445 -0.9895>}
 deform_wrist_dn_base_r {parent=hand_r, loc=<-0.0001 1.5328 0>, rot=<0 0 0 1>}
 deform_wrist_dn_r {parent=deform_wrist_dn_base_r, loc=<0 0 0>, rot=<0 0 0 1>}
 deform_wrist_up_base_r {parent=hand_r, loc=<-0.0002 -2.3007 -0.0001>, rot=<0 0 0 1>}
 deform_wrist_up_r {parent=deform_wrist_up_base_r, loc=<0 0 0>, rot=<0 0 0 1>}
 weapon_r {parent=hand_r, loc=<0.1974 3.4111 -0.3811>, rot=<0 0 0 1>}
 hand_attach_r {parent=hand_r, loc=<-8.5001 2.9999 0>, rot=<0 0 -0.7071 -0.7071>}
 lowerarm_twist_02_r {parent=lowerarm_r, loc=<-10.6707 0 0>, rot=<0 0 0 1>}
 lowerarm_twist_01_r {parent=lowerarm_r, loc=<-18.316 0 0>, rot=<0 0 0 1>}
 base_r_cuff_t {parent=lowerarm_twist_01_r, loc=<-1.5137 2.236 -2.712>, rot=<0 0 -0.7373 -0.6756>}
 dyn_r_cuff_t {parent=base_r_cuff_t, loc=<0 0 0>, rot=<0 0 0 1>}
 base_r_cuff_b {parent=lowerarm_twist_01_r, loc=<-2.6707 2.3395 4.4934>, rot=<0 0 -0.7373 -0.6756>}
 dyn_r_cuff_b {parent=base_r_cuff_b, loc=<0 0 0>, rot=<0 0 0 1>}
 deform_elbow_in_base_r {parent=lowerarm_r, loc=<-0.1834 -3.6062 0.2572>, rot=<0 0 0 -1>}
 deform_elbow_in_r {parent=deform_elbow_in_base_r, loc=<0 0 0>, rot=<0 0 0 1>}
 deform_elbow_out_base_r {parent=lowerarm_r, loc=<0 4 -0.0001>, rot=<0 0 0 -1>}
 deform_elbow_out_r {parent=deform_elbow_out_base_r, loc=<0 0 0>, rot=<0 0 0 1>}
 upperarm_twist_01_r {parent=upperarm_r, loc=<-7.2135 0.1134 0>, rot=<0 0 0 1>}
 deform_upperarm_bk_base_r {parent=upperarm_twist_01_r, loc=<6.4053 4.3746 -0.314>, rot=<0 0 0 -1>}
 deform_upperarm_bk_r {parent=deform_upperarm_bk_base_r, loc=<0 0 0>, rot=<0 0 0 1>}
 deform_upperarm_fr_base_r {parent=upperarm_twist_01_r, loc=<5.6312 -5.092 -0.3354>, rot=<0 0 0 1>}
 deform_upperarm_fr_r {parent=deform_upperarm_fr_base_r, loc=<0 0 0>, rot=<0 0 0 1>}
 deform_upperarm_up_base_r {parent=upperarm_twist_01_r, loc=<8.891 -0.4825 -3.9517>, rot=<0 0 0 1>}
 deform_upperarm_up_r {parent=deform_upperarm_up_base_r, loc=<0 0 0>, rot=<0 0 0 -1>}
 upperarm_twist_02_r {parent=upperarm_r, loc=<-19.7998 0.1134 0>, rot=<0 0 0 1>}
 base_lapel_r {parent=spine_05, loc=<10.2847 3.3675 5.5765>, rot=<0.2982 0.0388 0.9454 0.1259>}
 dyn_lapel_r {parent=base_lapel_r, loc=<0 0 0>, rot=<0 0 0 1>}
 dyn_lapel_01_r {parent=dyn_lapel_r, loc=<5.3912 -0.0035 -0.3695>, rot=<0 0 0 1>}
 base_lapel_l {parent=spine_05, loc=<10.4722 3.3393 -5.4271>, rot=<-0.2132 -0.0223 0.9612 0.1736>}
 dyn_lapel_l {parent=base_lapel_l, loc=<0 0 0>, rot=<0 0 0 1>}
 dyn_lapel_01_r1 {parent=dyn_lapel_l, loc=<5.7613 0.2196 -0.7932>, rot=<0 0 0 1>}
 dyn_lapel_01_l {parent=dyn_lapel_l, loc=<5.7613 0.2196 -0.7932>, rot=<0 0 0 1>}
 base_t_zipper_l {parent=spine_05, loc=<-2.1171 9.9603 -8.5189>, rot=<-0.0007 -0.5 0.866 0.0013>}
 dyn_t_zipper_l {parent=base_t_zipper_l, loc=<0 0 0>, rot=<0 0 0 1>}
 base_t_zipper_r {parent=spine_05, loc=<-2.1185 9.9775 8.5429>, rot=<0.0007 0.5 0.866 0.0013>}
 dyn_t_zipper_r {parent=base_t_zipper_r, loc=<0 0 0>, rot=<0 0 0 1>}
 deform_pec_aim_base_l {parent=spine_04, loc=<3.2456 9.3309 -1.2237>, rot=<0.0603 -0.4033 0.3318 -0.8507>}
 deform_pec_aim_l {parent=deform_pec_aim_base_l, loc=<0 0 0>, rot=<0 0 0 1>}
 deform_pec_base_l {parent=deform_pec_aim_l, loc=<9.3665 0 0>, rot=<0 0 0 1>}
 deform_pec_l {parent=deform_pec_base_l, loc=<0 0 0>, rot=<0 0 0 1>}
 deform_pec_aim_base_r {parent=spine_04, loc=<3.2452 9.3309 1.2237>, rot=<-0.4033 -0.0603 0.8507 0.3318>}
 deform_pec_aim_r {parent=deform_pec_aim_base_r, loc=<0 0 0>, rot=<0 0 0 1>}
 deform_pec_base_r {parent=deform_pec_aim_r, loc=<-9.3667 -0.0003 -0.0004>, rot=<0 0 0 1>}
 deform_pec_r {parent=deform_pec_base_r, loc=<0 0 0>, rot=<0 0 0 1>}
 base_f_jacket_r {parent=spine_04, loc=<-3.3478 10.2442 8.4031>, rot=<0 0.4226 0.9063 0>}
 dyn_f_jacket_r {parent=base_f_jacket_r, loc=<0 0 0>, rot=<0 0 0 1>}
 dyn_f_jacket_01_r {parent=dyn_f_jacket_r, loc=<9.2466 0.5325 -0.496>, rot=<0 0 0 1>}
 dyn_f_jacket_02_r {parent=dyn_f_jacket_01_r, loc=<7.5884 0.3051 -0.3636>, rot=<-0.0436 0 0 -0.999>}
 base_lower_zipper_r {parent=dyn_f_jacket_02_r, loc=<6.5747 -2.8504 0.4408>, rot=<-0.2588 0 0 -0.9659>}
 dyn_lower_zipper_r {parent=base_lower_zipper_r, loc=<-0.0001 0 -0.0001>, rot=<0 0 0 -1>}
 base_lo_zipper_r {parent=dyn_f_jacket_01_r, loc=<1.4588 -0.3682 0.7549>, rot=<-0.0436 0 0 -0.999>}
 dyn_lo_zipper_r {parent=base_lo_zipper_r, loc=<0 0 0>, rot=<0 0 0 1>}
 base_m_zipper_r {parent=dyn_f_jacket_r, loc=<3.7489 -0.2611 0.6322>, rot=<-0.0436 0 0 -0.999>}
 dyn_m_zipper_r {parent=base_m_zipper_r, loc=<0 0 0>, rot=<0 0 0 1>}
 base_f_jacket_l {parent=spine_04, loc=<-3.5827 10.1551 -8.6394>, rot=<0 -0.5 0.866 0>}
 dyn_f_jacket_l {parent=base_f_jacket_l, loc=<0 0 0>, rot=<0 0 0 1>}
 dyn_f_jacket_01_l {parent=dyn_f_jacket_l, loc=<9.2627 0.6792 0.5429>, rot=<0 0 0 1>}
 dyn_f_jacket_02_l {parent=dyn_f_jacket_01_l, loc=<7.4621 0.3112 0.539>, rot=<0 0 0 1>}
 base_lo_zipper_l {parent=dyn_f_jacket_01_l, loc=<1.3305 -0.8873 -0.6519>, rot=<0 0 0 1>}
 dyn_lo_zipper_l {parent=base_lo_zipper_l, loc=<0 0 0>, rot=<0 0 0 1>}
 base_m_zipper_l {parent=dyn_f_jacket_l, loc=<3.3025 -0.5164 -0.6613>, rot=<0 0 0 1>}
 dyn_m_zipper_l {parent=base_m_zipper_l, loc=<0 0 0>, rot=<0 0 0 1>}
 base_jacket_b {parent=spine_02, loc=<2.0329 -10.4945 0.1544>, rot=<-0.0012 -0.7071 0.7071 0.0012>}
 dyn_jacket_b {parent=base_jacket_b, loc=<0 0 0>, rot=<0 0 0 1>}
 base_jacket_l {parent=spine_02, loc=<-0.9997 -0.0649 -12.1351>, rot=<-0.0012 -0.7071 0.7071 0.0012>}
 dyn_jacket_l {parent=base_jacket_l, loc=<0 0 0>, rot=<0 0 0 1>}
 base_jacket_r {parent=spine_02, loc=<-0.9997 -0.0649 12.1252>, rot=<-0.0012 -0.7071 0.7071 0.0012>}
 dyn_jacket_r {parent=base_jacket_r, loc=<0 0 0>, rot=<0 0 0 1>}
 thigh_l {parent=pelvis, loc=<-1.0855 -0.2109 -8.7053>, rot=<0.0009 0.0355 0.0155 -0.9993>}
 calf_l {parent=thigh_l, loc=<-42.5284 -0.0019 -0.7297>, rot=<0.0006 -0.0161 -0.0391 -0.9991>}
 foot_l {parent=calf_l, loc=<-46.7401 -2.5014 -0.592>, rot=<0.0427 -0.0238 0.0353 -0.9982>}
 ball_l {parent=foot_l, loc=<-3.5658 10.8066 -0.1544>, rot=<0.0001 -0.0001 -0.7063 -0.7079>}
 calf_twist_02_l {parent=calf_l, loc=<-19.1546 -1.7943 -0.2638>, rot=<0.0237 0.0095 -0.0035 -0.9997>}
 calf_twist_01_l {parent=calf_l, loc=<-35.8483 -1.9197 -0.5794>, rot=<0.0237 0.0095 -0.0035 -0.9997>}
 deform_knee_in_base_l {parent=calf_l, loc=<0.8031 -5.0231 0.2456>, rot=<-0.0309 -0.7201 0.0276 -0.6927>}
 deform_knee_in_l {parent=deform_knee_in_base_l, loc=<0 0 0>, rot=<0 0 0 1>}
 deform_knee_out_base_l {parent=calf_l, loc=<0 4 0>, rot=<0 0 0 -1>}
 deform_knee_out_l {parent=deform_knee_out_base_l, loc=<0 0 0>, rot=<0 0 0 1>}
 foot_l_null {parent=calf_l, type=Null, loc=<-42.561 0.1077 0.4292>, rot=<0.0418 -0.0455 0.0343 -0.9975>}
 thigh_twist_01_l {parent=thigh_l, loc=<-27.952 0 0>, rot=<0.0013 0 0 -1>}
 deform_groin_base_l {parent=thigh_l, loc=<0.4801 5.6776 -0.2486>, rot=<-0.0019 -0.7317 0.0015 -0.6816>}
 deform_groin_l {parent=deform_groin_base_l, loc=<0 0 0>, rot=<0 0 0 1>}
 calf_l_null {parent=thigh_l, type=Null, loc=<-39.0484 0 0>, rot=<0.0006 -0.0161 -0.0391 -0.9991>}
 thigh_r {parent=pelvis, loc=<-1.0855 -0.2109 8.7053>, rot=<0.0355 -0.0009 0.9993 0.0155>}
 calf_r {parent=thigh_r, loc=<42.5284 0.0019 0.7297>, rot=<0.0006 -0.0161 -0.0391 -0.9991>}
 foot_r {parent=calf_r, loc=<46.7401 2.5014 0.592>, rot=<0.0427 -0.0238 0.0353 -0.9982>}
 ball_r {parent=foot_r, loc=<3.5658 -10.8066 0.1544>, rot=<0.0001 -0.0001 -0.7063 -0.7079>}
 calf_twist_02_r {parent=calf_r, loc=<19.1546 1.7943 0.2638>, rot=<0.0237 0.0095 -0.0035 -0.9997>}
 calf_twist_01_r {parent=calf_r, loc=<35.8483 1.9197 0.5794>, rot=<0.0237 0.0095 -0.0035 -0.9997>}
 deform_knee_in_base_r {parent=calf_r, loc=<-0.8031 5.0231 -0.2457>, rot=<0 0 0 -1>}
 deform_knee_in_r {parent=deform_knee_in_base_r, loc=<0 0 0>, rot=<0 0 0 1>}
 deform_knee_out_base_r {parent=calf_r, loc=<0 -4 0>, rot=<0 0 0 -1>}
 deform_knee_out_r {parent=deform_knee_out_base_r, loc=<0 0 0>, rot=<0 0 0 1>}
 foot_r_null {parent=calf_r, type=Null, loc=<42.561 -0.1077 -0.4292>, rot=<0.0418 -0.0455 0.0343 -0.9975>}
 thigh_twist_01_r {parent=thigh_r, loc=<27.952 0 0>, rot=<0.0013 0 0 -1>}
 deform_groin_base_r {parent=thigh_r, loc=<-0.4801 -5.6776 0.2486>, rot=<0 0 0 -1>}
 deform_groin_r {parent=deform_groin_base_r, loc=<0 0 0>, rot=<0 0 0 1>}
 calf_r_null {parent=thigh_r, type=Null, loc=<39.0484 0 0>, rot=<0.0006 -0.0161 -0.0391 -0.9991>}
 leg_pivot {parent=thigh_r, loc=<13.609 -1.9139 7.8567>, rot=<0.5375 -0.0535 0.018 -0.8414>}
 deform_glute_base_l {parent=pelvis, loc=<-7.3466 -9.0213 -9.1245>, rot=<0.0172 0.031 -0.4477 -0.8935>}
 deform_glute_l {parent=deform_glute_base_l, loc=<0 0 0>, rot=<0 0 0 1>}
 deform_glute_base_r {parent=pelvis, loc=<-7.3464 -9.0213 9.1245>, rot=<0.031 -0.0172 0.8935 -0.4477>}
 deform_glute_r {parent=deform_glute_base_r, loc=<0 0 0>, rot=<0 0 0 1>}
 thigh_r_null {parent=pelvis, type=Null, loc=<-2.5693 -0.0109 7.7501>, rot=<0.0572 -0.0012 0.9982 0.0155>}
 thigh_l_null {parent=pelvis, type=Null, loc=<-2.5693 -0.0109 -7.7501>, rot=<0.0012 0.0572 0.0155 -0.9982>}
 base_knife {parent=pelvis, loc=<2.8926 2.0308 14.1704>, rot=<-0.032 -0.5363 0.8419 0.0502>}
 dyn_knife {parent=base_knife, loc=<0 0 0>, rot=<0 0 0 1>}
 bottom_kinfe {parent=base_knife, loc=<11.9 -1.1865 2.2575>, rot=<0 0 0 1>}
 base_knife_null {parent=pelvis, type=Null, loc=<2.8926 2.0308 14.1704>, rot=<-0.032 -0.5363 0.8419 0.0502>}
]
variables [
 Crouch {type=double, default=0}
 CrouchMoving {type=double, default=0}
 FE {type=bool, default=False}
 JUMP {type=double, default=0}
 Skydiving {type=double, default=0}
 Target {type=double, default=0}
 crouchSwing {type=double, default=0}
 dbno {type=double, default=0}
 parachute {type=double, default=0}
]
