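# Post-process style AnimBlueprint for female SurgeRaven ZanyOwl body, adding secondary motion to jacket, zippers, cuffs and lapels.
# Copies the parent mesh pose into the male variant's Control Rig, driven by crouch and DBNO values.
# Pelvis-space RigidBody, weighted zipper-base constraints, AnimDynamics and a right-jacket bone offset follow.
# Event graph updates state floats, gravity override and RBAN settings; only gravity reaches RigidBody, and Melee is unused.
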AnimBlueprint /BRCosmetics/Characters/Player/Female/Medium/Bodies/F_MED_SurgeRaven_ZanyOwl/Meshes/F_MED_SurgeRaven_ZanyOwl_AnimBP
target_skeleton=&/F_MED_SurgeRaven_ZanyOwl_Skeleton
graphs [
 AnimGraph {nodes [
   n1 = Control Rig!(
    Source=Copy Pose From Mesh() {bUseAttachedParent=true}
    Alpha
    Crouch:double=Get crouch() -> (crouch:double)
    dbno:double=Get DBNO() -> (DBNO:double)
    CrouchMoving:double=Get CrouchMoving() -> (CrouchMoving:double)
   ) -> (Pose) {controlRigAssetReference {blueprintRigClass=&/M_MED_SurgeRaven_ZanyOwl_CtrlRig.M_MED_SurgeRaven_ZanyOwl_CtrlRig_C}}
   n2 = RigidBody!(ComponentPose=Local To Component(LocalPose=@n1), Alpha, OverrideWorldGravity:Vector=Get Gravity Override(), ExternalForce) -> (Pose) {
    bOverrideWorldGravity=true
    overridePhysicsAsset=&/F_MED_SurgeRaven_ZanyOwl_Physics
    simulationSpace=BaseBoneSpace
    alphaBoolBlend {blendOption=Linear}
    alphaScaleBiasClamp {bInterpResult=true}
    baseBoneRef {boneName=pelvis}
    componentAppliedLinearAccClamp=<0 0 0>
    overrideWorldGravity {z=-980}
    simSpaceSettings {
     dampingAlpha=0
     maxAngularAcceleration=5
     maxAngularVelocity=20
     maxLinearAcceleration=3000
     maxLinearVelocity=600
     velocityScaleZ=0.2
     worldAlpha=0.7
     externalLinearDragV=<0.5 0.5 0.5>
    }
   }
   n3 = "Constraint - base_m_zipper_l (PPPPPP)"!(
    ComponentPose=@n2
    Alpha=1
    ConstraintWeights_0=0.7362
    ConstraintWeights_1:float=0.1314
    ConstraintWeights_2:float=0.1189
    ConstraintWeights_3:float=0.0052
    ConstraintWeights_4:float=0.0046
    ConstraintWeights_5:float=0.0036
   ) -> (Pose) {
    boneToModify {boneName=base_m_zipper_l}
    constraintSetup [
     {targetBone {boneName=dyn_f_jacket_l}}
     {targetBone {boneName=spine_04}}
     {targetBone {boneName=dyn_f_jacket_01_l}}
     {targetBone {boneName=dyn_f_jacket_02_l}}
     {targetBone {boneName=spine_02}}
     {targetBone {boneName=spine_05}}
    ]
    constraintWeights [1,1,1,1,1,1]
   }
   n4 = "Constraint - base_lo_zipper_l (PPPPPPP)"!(
    ComponentPose=@n3
    Alpha=1
    ConstraintWeights_0=0.3466
    ConstraintWeights_1:float=0.2673
    ConstraintWeights_2:float=0.2629
    ConstraintWeights_3:float=0.0613
    ConstraintWeights_4:float=0.0298
    ConstraintWeights_5:float=0.0202
    ConstraintWeights_6:float=0.0118
   ) -> (Pose) {
    boneToModify {boneName=base_lo_zipper_l}
    constraintSetup [
     {targetBone {boneName=dyn_f_jacket_01_l}}
     {targetBone {boneName=dyn_f_jacket_02_l}}
     {targetBone {boneName=dyn_f_jacket_l}}
     {targetBone {boneName=spine_04}}
     {targetBone {boneName=dyn_jacket_l}}
     {targetBone {boneName=pelvis}}
     {targetBone {boneName=spine_02}}
    ]
    constraintWeights [1,1,1,1,1,1,1]
   }
   n5 = "Constraint - base_lo_zipper_r (PPPPPPPP)"!(
    ComponentPose=@n4
    Alpha=1
    ConstraintWeights_0=0.3515
    ConstraintWeights_1:float=0.2677
    ConstraintWeights_2:float=0.2575
    ConstraintWeights_3:float=0.0628
    ConstraintWeights_4:float=0.0285
    ConstraintWeights_5:float=0.0195
    ConstraintWeights_6:float=0.0117
    ConstraintWeights_7:float=0.0009
   ) -> (Pose) {
    boneToModify {boneName=base_lo_zipper_r}
    constraintSetup [
     {targetBone {boneName=dyn_f_jacket_01_r}}
     {targetBone {boneName=dyn_f_jacket_r}}
     {targetBone {boneName=dyn_f_jacket_02_r}}
     {targetBone {boneName=spine_04}}
     {targetBone {boneName=dyn_jacket_r}}
     {targetBone {boneName=pelvis}}
     {targetBone {boneName=spine_02}}
     {targetBone {boneName=spine_01}}
    ]
    constraintWeights [1,1,1,1,1,1,1,1]
   }
   n6 = "Constraint - base_lower_zipper_r (PPPPPPP)"!(
    ComponentPose=@n5
    Alpha=1
    ConstraintWeights_0=0.9907
    ConstraintWeights_1:float=0.0031
    ConstraintWeights_2:float=0.003
    ConstraintWeights_3:float=0.0024
    ConstraintWeights_4:float=0.0006
    ConstraintWeights_5:float=0.0002
    ConstraintWeights_6:float=0.0001
   ) -> (Pose) {
    boneToModify {boneName=base_lower_zipper_r}
    constraintSetup [
     {targetBone {boneName=dyn_f_jacket_02_r}}
     {targetBone {boneName=dyn_f_jacket_01_r}}
     {targetBone {boneName=dyn_f_jacket_r}}
     {targetBone {boneName=dyn_jacket_r}}
     {targetBone {boneName=pelvis}}
     {targetBone {boneName=deform_elbow_in_base_r}}
     {targetBone {boneName=spine_02}}
    ]
    constraintWeights [1,1,1,1,1,1,1]
   }
   n7 = "Constraint - base_m_zipper_r (PPPPPPPP)"!(
    ComponentPose=@n6
    Alpha=1
    ConstraintWeights_0=0.7394
    ConstraintWeights_1:float=0.1297
    ConstraintWeights_2:float=0.1128
    ConstraintWeights_3:float=0.0052
    ConstraintWeights_4:float=0.0042
    ConstraintWeights_5:float=0.004
    ConstraintWeights_6:float=0.0034
    ConstraintWeights_7:float=0.0013
   ) -> (Pose) {
    boneToModify {boneName=base_m_zipper_r}
    constraintSetup [
     {targetBone {boneName=dyn_f_jacket_r}}
     {targetBone {boneName=spine_04}}
     {targetBone {boneName=dyn_f_jacket_01_r}}
     {targetBone {boneName=dyn_f_jacket_02_r}}
     {targetBone {boneName=spine_02}}
     {targetBone {boneName=spine_05}}
     {targetBone {boneName=spine_03}}
     {targetBone {boneName=spine_01}}
    ]
    constraintWeights [1,1,1,1,1,1,1,1]
   }
   n8 = "Constraint - base_t_zipper_r (PPP)"!(ComponentPose=@n7, Alpha=1, ConstraintWeights_0=0.8517, ConstraintWeights_1:float=0.1317, ConstraintWeights_2:float=0.0166) -> (Pose) {
    boneToModify {boneName=base_t_zipper_r}
    constraintSetup [{targetBone {boneName=spine_04}}, {targetBone {boneName=spine_05}}, {targetBone {boneName=clavicle_r}}]
    constraintWeights [1,1,1]
   }
   n9 = "Constraint - base_t_zipper_l (PPPP)"!(
    ComponentPose=@n8
    Alpha=1
    ConstraintWeights_0=0.8514
    ConstraintWeights_1:float=0.1308
    ConstraintWeights_2:float=0.0177
    ConstraintWeights_3:float=0.0001
   ) -> (Pose) {
    boneToModify {boneName=base_t_zipper_l}
    constraintSetup [
     {targetBone {boneName=spine_04}}
     {targetBone {boneName=spine_05}}
     {targetBone {boneName=clavicle_l}}
     {targetBone {boneName=upperarm_twist_01_l}}
    ]
    constraintWeights [1,1,1,1]
   }
   n10 = Anim Dynamics - Bone: dyn_m_zipper_r(ComponentPose=@n9, Alpha=1, ExternalForce="0,0,0") {
    angularDampingOverride=1
    angularSpringConstant=1000
    bAngularSpring=true
    bOverrideAngularDamping=true
    bUseGravityOverride=true
    gravityScale=0.35
    simulationSpace=BoneRelative
    boundBone {boneName=dyn_m_zipper_r}
    componentLinearAccScale=<0.5 0.5 0.5>
    componentLinearVelScale=<0.5 0.5 0.5>
    gravityOverride {z=-980}
    physicsBodyDefinitions [{
      boundBone {boneName=dyn_m_zipper_r}
      boxExtents=<20 30 30>
      constraintSetup {angularLimitsMax=<15 0 0>, angularLimitsMin=<-15 -80 -40>, angularTarget=<1 0 0>}
      localJointOffset=<4 0 0>
     }]
    simSpaceSettings {simSpaceAngularAlpha=0.5}
   }
   n11 = Anim Dynamics - Bone: dyn_m_zipper_l(ComponentPose=@n10, Alpha=1, ExternalForce="0,0,0") {
    angularDampingOverride=1
    angularSpringConstant=1000
    bAngularSpring=true
    bOverrideAngularDamping=true
    bUseGravityOverride=true
    gravityScale=0.35
    simulationSpace=BoneRelative
    boundBone {boneName=dyn_m_zipper_l}
    componentLinearAccScale=<0.5 0.5 0.5>
    componentLinearVelScale=<0.5 0.5 0.5>
    gravityOverride {z=-980}
    physicsBodyDefinitions [{
      boundBone {boneName=dyn_m_zipper_l}
      boxExtents=<20 30 30>
      constraintSetup {angularLimitsMax=<15 80 0>, angularLimitsMin=<-15 10 -40>, angularTarget=<1 0 0>}
      localJointOffset=<4 0 0>
     }]
    simSpaceSettings {simSpaceAngularAlpha=0.5}
   }
   n12 = Anim Dynamics - Bone: dyn_lo_zipper_l(ComponentPose=@n11, Alpha=1, ExternalForce="0,0,0") {
    angularDampingOverride=1
    angularSpringConstant=1000
    bAngularSpring=true
    bOverrideAngularDamping=true
    bUseGravityOverride=true
    gravityScale=0.35
    simulationSpace=BoneRelative
    boundBone {boneName=dyn_lo_zipper_l}
    componentLinearAccScale=<0.5 0.5 0.5>
    componentLinearVelScale=<0.5 0.5 0.5>
    gravityOverride {z=-980}
    physicsBodyDefinitions [{
      boundBone {boneName=dyn_lo_zipper_l}
      boxExtents=<20 30 30>
      constraintSetup {angularLimitsMax=<15 80 0>, angularLimitsMin=<-15 10 -40>, angularTarget=<1 0 0>}
      localJointOffset=<4 0 0>
     }]
    simSpaceSettings {simSpaceAngularAlpha=0.5}
   }
   n13 = Anim Dynamics - Bone: dyn_lo_zipper_r(ComponentPose=@n12, Alpha=1, ExternalForce="0,0,0") {
    angularDampingOverride=1
    angularSpringConstant=1000
    bAngularSpring=true
    bOverrideAngularDamping=true
    bUseGravityOverride=true
    gravityScale=0.35
    simulationSpace=BoneRelative
    boundBone {boneName=dyn_lo_zipper_r}
    componentLinearAccScale=<0.5 0.5 0.5>
    componentLinearVelScale=<0.5 0.5 0.5>
    gravityOverride {z=-980}
    physicsBodyDefinitions [{
      boundBone {boneName=dyn_lo_zipper_r}
      boxExtents=<20 30 30>
      constraintSetup {angularLimitsMax=<15 0 0>, angularLimitsMin=<-15 -80 -40>, angularTarget=<1 0 0>}
      localJointOffset=<4 0 0>
     }]
    simSpaceSettings {simSpaceAngularAlpha=0.5}
   }
   n14 = Anim Dynamics - Bone: dyn_t_zipper_l(ComponentPose=@n13, Alpha=1, ExternalForce="0,0,0") {
    angularDampingOverride=1
    angularSpringConstant=1000
    bAngularSpring=true
    bOverrideAngularDamping=true
    bUseGravityOverride=true
    gravityScale=0.35
    simulationSpace=RootRelative
    boundBone {boneName=dyn_t_zipper_l}
    componentLinearAccScale=<0.5 0.5 0.5>
    componentLinearVelScale=<0.5 0.5 0.5>
    gravityOverride {z=-980}
    physicsBodyDefinitions [{
      boundBone {boneName=dyn_t_zipper_l}
      boxExtents=<20 30 30>
      constraintSetup {angularLimitsMax=<15 80 0>, angularLimitsMin=<-15 0 -40>, angularTarget=<1 0 0>}
      localJointOffset=<4 0 0>
     }]
    simSpaceSettings {simSpaceAngularAlpha=0.5}
   }
   n15 = Anim Dynamics - Bone: dyn_t_zipper_r(ComponentPose=@n14, Alpha=1, ExternalForce="0,0,0") {
    angularDampingOverride=1
    angularSpringConstant=1000
    bAngularSpring=true
    bOverrideAngularDamping=true
    bUseGravityOverride=true
    gravityScale=0.35
    simulationSpace=RootRelative
    boundBone {boneName=dyn_t_zipper_r}
    componentLinearAccScale=<0.5 0.5 0.5>
    componentLinearVelScale=<0.5 0.5 0.5>
    gravityOverride {z=-980}
    physicsBodyDefinitions [{
      boundBone {boneName=dyn_t_zipper_r}
      boxExtents=<20 30 30>
      constraintSetup {angularLimitsMax=<15 0 0>, angularLimitsMin=<-15 -80 -40>, angularTarget=<1 0 0>}
      localJointOffset=<4 0 0>
     }]
    simSpaceSettings {simSpaceAngularAlpha=0.5}
   }
   n16 = Anim Dynamics - Bone: dyn_lower_zipper_r(ComponentPose=@n15, Alpha=1, ExternalForce="0,0,0") {
    angularDampingOverride=1
    angularSpringConstant=1000
    bAngularSpring=true
    bOverrideAngularDamping=true
    bUseGravityOverride=true
    gravityScale=0.35
    simulationSpace=BoneRelative
    boundBone {boneName=dyn_lower_zipper_r}
    componentLinearAccScale=<0.5 0.5 0.5>
    componentLinearVelScale=<0.5 0.5 0.5>
    gravityOverride {z=-980}
    physicsBodyDefinitions [{
      boundBone {boneName=dyn_lower_zipper_r}
      boxExtents=<20 30 30>
      constraintSetup {angularLimitsMax=<15 30 0>, angularLimitsMin=<-15 -80 -40>, angularTarget=<1 0 0>}
      localJointOffset=<4 0 0>
     }]
    simSpaceSettings {simSpaceAngularAlpha=0.5}
   }
   n17 = Anim Dynamics - Bone: dyn_l_cuff_t(ComponentPose=@n16, Alpha=1, ExternalForce="0,0,0") {
    angularDampingOverride=1
    angularSpringConstant=10000
    bAngularSpring=true
    bOverrideAngularDamping=true
    bUseGravityOverride=true
    gravityScale=0.35
    simulationSpace=RootRelative
    boundBone {boneName=dyn_l_cuff_t}
    componentLinearAccScale=<1 1 1>
    componentLinearVelScale=<1 1 1>
    gravityOverride {z=-980}
    physicsBodyDefinitions [{
      boundBone {boneName=dyn_l_cuff_t}
      boxExtents=<20 30 30>
      constraintSetup {angularLimitsMax=<15 15 15>, angularLimitsMin=<-15 -15 -15>, angularTarget=<1 0 0>}
      localJointOffset=<4 0 0>
     }]
   }
   n18 = Anim Dynamics - Bone: dyn_l_cuff_b(ComponentPose=@n17, Alpha=1, ExternalForce="0,0,0") {
    angularDampingOverride=1
    angularSpringConstant=10000
    bAngularSpring=true
    bOverrideAngularDamping=true
    bUseGravityOverride=true
    gravityScale=0.35
    simulationSpace=RootRelative
    boundBone {boneName=dyn_l_cuff_b}
    componentLinearAccScale=<1 1 1>
    componentLinearVelScale=<1 1 1>
    gravityOverride {z=-980}
    physicsBodyDefinitions [{
      boundBone {boneName=dyn_l_cuff_b}
      boxExtents=<20 30 30>
      constraintSetup {angularLimitsMax=<15 15 15>, angularLimitsMin=<-15 -15 -15>, angularTarget=<1 0 0>}
      localJointOffset=<4 0 0>
     }]
   }
   n19 = Anim Dynamics - Bone: dyn_r_cuff_t(ComponentPose=@n18, Alpha=1, ExternalForce="0,0,0") {
    angularDampingOverride=1
    angularSpringConstant=10000
    bAngularSpring=true
    bOverrideAngularDamping=true
    bUseGravityOverride=true
    gravityScale=0.35
    simulationSpace=RootRelative
    boundBone {boneName=dyn_r_cuff_t}
    componentLinearAccScale=<1 1 1>
    componentLinearVelScale=<1 1 1>
    gravityOverride {z=-980}
    physicsBodyDefinitions [{
      boundBone {boneName=dyn_r_cuff_t}
      boxExtents=<20 30 30>
      constraintSetup {angularLimitsMax=<15 15 15>, angularLimitsMin=<-15 -15 -15>, angularTarget=<1 0 0>}
      localJointOffset=<4 0 0>
     }]
   }
   n20 = Anim Dynamics - Bone: dyn_r_cuff_b(ComponentPose=@n19, Alpha=1, ExternalForce="0,0,0") {
    angularDampingOverride=1
    angularSpringConstant=10000
    bAngularSpring=true
    bOverrideAngularDamping=true
    bUseGravityOverride=true
    gravityScale=0.35
    simulationSpace=RootRelative
    boundBone {boneName=dyn_r_cuff_b}
    componentLinearAccScale=<1 1 1>
    componentLinearVelScale=<1 1 1>
    gravityOverride {z=-980}
    physicsBodyDefinitions [{
      boundBone {boneName=dyn_r_cuff_b}
      boxExtents=<20 30 30>
      constraintSetup {angularLimitsMax=<15 15 15>, angularLimitsMin=<-15 -15 -15>, angularTarget=<1 0 0>}
      localJointOffset=<4 0 0>
     }]
   }
   n21 = Anim Dynamics - Bone: dyn_lapel_01_r(ComponentPose=@n20, Alpha=1, ExternalForce="0,0,0") {
    angularDampingOverride=1
    angularSpringConstant=10000
    bAngularSpring=true
    bOverrideAngularDamping=true
    bUseGravityOverride=true
    gravityScale=0.35
    simulationSpace=BoneRelative
    boundBone {boneName=dyn_lapel_01_r}
    componentLinearAccScale=<1 1 1>
    componentLinearVelScale=<1 1 1>
    gravityOverride {z=-980}
    physicsBodyDefinitions [{
      boundBone {boneName=dyn_lapel_01_r}
      boxExtents=<20 30 30>
      constraintSetup {angularLimitsMax=<15 15 0>, angularLimitsMin=<-15 -15 -30>, angularTarget=<1 0 0>}
      localJointOffset=<4 0 0>
     }]
   }
   n22 = Anim Dynamics - Bone: dyn_lapel_01_l(ComponentPose=@n21, Alpha=1, ExternalForce="0,0,0") {
    angularDampingOverride=1
    angularSpringConstant=10000
    bAngularSpring=true
    bOverrideAngularDamping=true
    bUseGravityOverride=true
    gravityScale=0.35
    simulationSpace=BoneRelative
    boundBone {boneName=dyn_lapel_01_l}
    componentLinearAccScale=<1 1 1>
    componentLinearVelScale=<1 1 1>
    gravityOverride {z=-980}
    physicsBodyDefinitions [{
      boundBone {boneName=dyn_lapel_01_l}
      boxExtents=<20 30 30>
      constraintSetup {angularLimitsMax=<15 15 0>, angularLimitsMin=<-15 -15 -30>, angularTarget=<1 0 0>}
      localJointOffset=<4 0 0>
     }]
   }
   n23 = "Transform (Modify) Bone - Bone: base_jacket_r"(ComponentPose=@n22, Alpha=1, Translation="-1,-5,-1", Rotation="0,0,0", Scale="1,1,1") {
    translationMode=BMM_Additive
    translationSpace=BCS_BoneSpace
    alphaScaleBiasClamp {bInterpResult=true, interpSpeedDecreasing=5}
    boneToModify {boneName=base_jacket_r}
    translation=<-1 -5 -1>
   }
   Output Pose(Result=Component To Local(ComponentPose=@n23)) {layerGroup=None}
  ]}
 EventGraph {nodes [
   Event BlueprintUpdateAnimation()
   > n1 = Gravity Override State Function DEPRECATED(
    CharacterPartAnimInstance=Self-Reference()
    GameTimeInSeconds=GetGameTimeInSeconds()
    GravityOverrideParams=Get Gravity Override Params()
    RBAN Params=Get RBAN Params_0() -> (RBAN Params_0:RBANParams Struct Structure)
   )
   > Set Gravity Override(Gravity Override=@n1)
   > Set RBAN Settings Out!(execute, RBAN Settings Out:RBANSetting Struct Structure=@n1.RBAN SettingsOut) -> (
    then
    Output_Get_ComponentLinearAccScale_28_#1:Vector=~
    Output_Get_ComponentLinearVelScale_31_#2:Vector=~
    Output_Get_SimSpaceSettings_34_#3:Sim Space Settings Structure
   )
   n2 = NOT Boolean(A=Get bIsMoving2D())
   n3 = NOT Boolean(A=Get bIsPlayingMeleeAnim())
   n4 = Get bIsCrouching()
   > Set crouch(crouch:double="To Float (Boolean)"(InBool=AND Boolean!(A=@n4, B=@n3, C:bool=@n2) -> (ReturnValue))) -> (Output_Get:double)
   > Set DBNO(DBNO:double="To Float (Boolean)"(InBool=Get bIsDBNO())) -> (Output_Get:double)
   n5 = Get bIsCrouching()
   n6 = Get bIsMoving2D()
   > Set CrouchMoving(CrouchMoving:double="To Float (Boolean)"(InBool=AND Boolean!(A=@n5, B=@n6) -> (ReturnValue))) -> (Output_Get:double)
   > Set Melee(Melee:double="To Float (Boolean)"(InBool=Get bIsPlayingMeleeAnim())) -> (Output_Get:double)
  ]}
]
variables [
 crouch {type=number}
 CrouchMoving {type=number}
 DBNO {type=number}
 Gravity Override {type=Vector}
 Gravity Override Params {type=GravityOverrideParamsStruct}
 Melee {type=number}
 RBAN Params_0 {
  type=RBANParamsStruct
  default {
   dBNO {simSpaceSettings {
     maxAngularAcceleration=1000
     maxAngularVelocity=3
     maxLinearAcceleration=300
     maxLinearVelocity=700
     velocityScaleZ=0
     externalLinearDragV=<0 0.5 0>
    }}
   dBNO_Carried {simSpaceSettings {
     maxAngularAcceleration=1000
     maxAngularVelocity=3
     maxLinearAcceleration=300
     maxLinearVelocity=700
     velocityScaleZ=0
     externalLinearDragV=<0 0.5 0>
    }}
   emoteOrMelee {simSpaceSettings {
     maxAngularAcceleration=1000
     maxAngularVelocity=3
     maxLinearAcceleration=300
     maxLinearVelocity=700
     velocityScaleZ=0
     externalLinearDragV=<0 0.5 0>
    }}
   front_End {simSpaceSettings {
     maxAngularAcceleration=1000
     maxAngularVelocity=3
     maxLinearAcceleration=300
     maxLinearVelocity=700
     velocityScaleZ=0
     externalLinearDragV=<0 0.5 0>
    }}
   onGround_Crouch_Moving {simSpaceSettings {
     maxAngularAcceleration=1000
     maxAngularVelocity=3
     maxLinearAcceleration=300
     maxLinearVelocity=700
     velocityScaleZ=0
     externalLinearDragV=<0 0.5 0>
    }}
   onGround_Falling {simSpaceSettings {
     maxAngularAcceleration=1000
     maxAngularVelocity=3
     maxLinearAcceleration=300
     maxLinearVelocity=700
     velocityScaleZ=0
     externalLinearDragV=<0 0.5 0>
    }}
   onGround_Moving {simSpaceSettings {
     maxAngularAcceleration=1000
     maxAngularVelocity=3
     maxLinearAcceleration=300
     maxLinearVelocity=700
     velocityScaleZ=0
     externalLinearDragV=<0 0.5 0>
    }}
   onGround_Standing {simSpaceSettings {
     maxAngularAcceleration=1000
     maxAngularVelocity=3
     maxLinearAcceleration=300
     maxLinearVelocity=700
     velocityScaleZ=0
     externalLinearDragV=<0 0.5 0>
    }}
   rail_Grinding {simSpaceSettings {
     maxAngularAcceleration=1000
     maxAngularVelocity=3
     maxLinearAcceleration=300
     maxLinearVelocity=700
     velocityScaleZ=0
     worldAlpha=0.1
     externalLinearDragV=<0 0.5 0>
    }}
   riding_Moving {simSpaceSettings {
     maxAngularAcceleration=1000
     maxAngularVelocity=3
     maxLinearAcceleration=300
     maxLinearVelocity=700
     velocityScaleZ=0
     externalLinearDragV=<0 0.5 0>
    }}
   riding_Moving_Targeting {simSpaceSettings {
     maxAngularAcceleration=1000
     maxAngularVelocity=3
     maxLinearAcceleration=300
     maxLinearVelocity=700
     velocityScaleZ=0
     externalLinearDragV=<0 0.5 0>
    }}
   riding_Standing {simSpaceSettings {
     maxAngularAcceleration=1000
     maxAngularVelocity=3
     maxLinearAcceleration=300
     maxLinearVelocity=700
     velocityScaleZ=0
     externalLinearDragV=<0 0.5 0>
    }}
   skydive_DivingDown {simSpaceSettings {
     maxAngularAcceleration=1000
     maxAngularVelocity=3
     maxLinearAcceleration=300
     maxLinearVelocity=700
     velocityScaleZ=0
     externalLinearDragV=<0 0.5 0>
    }}
   skydive_FloatingDown {
    componentLinearAccScale=<0 0.5 0>
    componentLinearVelScale=<0 0.5 0>
    simSpaceSettings {
     maxAngularAcceleration=1000
     maxAngularVelocity=3
     maxLinearAcceleration=300
     maxLinearVelocity=700
     velocityScaleZ=0
     externalLinearDragV=<0 0.5 0>
    }
   }
   skydive_GoingUp {simSpaceSettings {
     maxAngularAcceleration=1000
     maxAngularVelocity=3
     maxLinearAcceleration=300
     maxLinearVelocity=700
     velocityScaleZ=0
     externalLinearDragV=<0 0.5 0>
    }}
   skydive_Parachuting {
    componentLinearAccScale=<0 0.1 0.3>
    componentLinearVelScale=<0 0.1 0.3>
    simSpaceSettings {
     maxAngularAcceleration=1000
     maxAngularVelocity=3
     maxLinearAcceleration=300
     maxLinearVelocity=700
     velocityScaleZ=0
     externalLinearDragV=<0 0.5 0>
    }
   }
   skydive_Surfing {simSpaceSettings {
     maxAngularAcceleration=1000
     maxAngularVelocity=3
     maxLinearAcceleration=300
     maxLinearVelocity=700
     velocityScaleZ=0
     externalLinearDragV=<0 0.5 0>
    }}
   swimming {simSpaceSettings {
     maxAngularAcceleration=1000
     maxAngularVelocity=3
     maxLinearAcceleration=300
     maxLinearVelocity=700
     velocityScaleZ=0
     externalLinearDragV=<0 0.5 0>
    }}
   tactical_Sprint {simSpaceSettings {
     maxAngularAcceleration=1000
     maxAngularVelocity=3
     maxLinearAcceleration=300
     maxLinearVelocity=700
     velocityScaleZ=0
     externalLinearDragV=<0 0.5 0>
    }}
   ziplining {simSpaceSettings {
     maxAngularAcceleration=1000
     maxAngularVelocity=3
     maxLinearAcceleration=300
     maxLinearVelocity=700
     velocityScaleZ=0
     externalLinearDragV=<0 0.5 0>
    }}
  }
 }
 RBAN Settings Out {
  type=RBANSettingStruct
  default {
   componentLinearAccScale=<0 0 0>
   componentLinearVelScale=<0 0 0>
   simSpaceSettings {
    dampingAlpha=1
    maxAngularAcceleration=1000
    maxAngularVelocity=3
    maxLinearAcceleration=300
    maxLinearVelocity=700
    velocityScaleZ=0
    worldAlpha=1
    externalAngularVelocity=<0 0 0>
    externalLinearDragV=<0 0.5 0>
    externalLinearVelocity=<0 0 0>
   }
  }
 }
]
